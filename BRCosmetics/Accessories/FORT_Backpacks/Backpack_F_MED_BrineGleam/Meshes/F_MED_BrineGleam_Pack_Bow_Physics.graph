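# Physics asset for the bow on the BrineGleam backpack cosmetic, defining bodies and joint limits for its dangling bow and lace bones.
# Left and right two-link bow chains and one-link lace chains each hang from a small MSTR body, using limited swing and tight twist, with slightly looser swing on the laces.
# Larger capsules on four backpack spine bones have no constraints, apparently serving as torso collision volumes for the chains.
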PhysicsAsset /BRCosmetics/Accessories/FORT_Backpacks/Backpack_F_MED_BrineGleam/Meshes/F_MED_BrineGleam_Pack_Bow_Physics
bodies {
 dyn_bow_01_l [{shapeType=Capsule, length=0.5, radius=2, shapeName=dyn_bow_01_l_capsule, center=<2.3926 0 -0.3778>}]
 dyn_bow_01_r [{shapeType=Capsule, length=0.5, radius=2, shapeName=dyn_bow_01_r_capsule, center=<1.7382 0 0>}]
 dyn_bow_02_l [{shapeType=Capsule, length=0.5, radius=3, shapeName=dyn_bow_02_l_capsule, center=<0.5 0 0>}]
 dyn_bow_02_r [{shapeType=Capsule, length=1, radius=3, shapeName=dyn_bow_02_r_capsule, center=<0.7 0 0>}]
 dyn_bow_l_MSTR [{shapeType=Capsule, length=0.5, radius=0.5, shapeName=dyn_bow_l_MSTR_capsule}]
 dyn_bow_r_MSTR [{shapeType=Capsule, length=0.5, radius=0.5, shapeName=dyn_bow_r_MSTR_capsule}]
 dyn_lace_01_l [{shapeType=Capsule, length=1.5, radius=1, shapeName=dyn_lace_01_l_capsule, center=<1 0 0.1994>}]
 dyn_lace_01_r [{shapeType=Capsule, length=1.5, radius=1, shapeName=dyn_lace_01_r_capsule, center=<1 0 0>}]
 dyn_lace_l_MSTR [{shapeType=Capsule, length=0.5, radius=0.5, shapeName=dyn_lace_l_MSTR_capsule}]
 dyn_lace_r_MSTR [{shapeType=Capsule, length=0.5, radius=0.5, shapeName=dyn_lace_r_MSTR_capsule}]
 spine_02_BrineGleam_backpack [
  {shapeType=Capsule, length=2.5219, radius=9.7317, center=<-2.6885 -1.2641 0>}
  {shapeType=Capsule, length=20.6875, radius=9.3125, center=<-1.8317 1.9362 0>, rotation=<rot 90 0 0>}
 ]
 spine_03_BrineGleam_backpack [
  {shapeType=Capsule, length=6.1863, radius=8.4854, center=<0.0574 0.4416 0>}
  {shapeType=Capsule, length=20.7188, radius=9.2812, center=<-1.2163 0.549 0>, rotation=<rot 90 180 -175.0732>}
 ]
 spine_04_BrineGleam_backpack [
  {shapeType=Capsule, length=9.0076, radius=10.3916, center=<5.6234 -0.6779 0>}
  {shapeType=Capsule, length=10.0312, radius=9.2812, center=<-1.2326 1.308 0>, rotation=<rot 90 0 0>}
 ]
 spine_05_BrineGleam_backpack [{shapeType=Capsule, length=9.678, radius=12.3571, center=<0.6969 1.91 0>}]
}
constraints [
 {bone1Name=dyn_bow_01_l, bone2Name=dyn_bow_l_MSTR, swing1LimitDegrees=20, swing2LimitDegrees=20, twistLimitDegrees=5}
 {bone1Name=dyn_bow_02_l, bone2Name=dyn_bow_01_l, swing1LimitDegrees=20, swing2LimitDegrees=20, twistLimitDegrees=5}
 {bone1Name=dyn_bow_01_r, bone2Name=dyn_bow_r_MSTR, swing1LimitDegrees=20, swing2LimitDegrees=20, twistLimitDegrees=5}
 {bone1Name=dyn_bow_02_r, bone2Name=dyn_bow_01_r, swing1LimitDegrees=20, swing2LimitDegrees=20, twistLimitDegrees=5}
 {bone1Name=dyn_lace_01_l, bone2Name=dyn_lace_l_MSTR, swing1LimitDegrees=30, swing2LimitDegrees=30, twistLimitDegrees=5}
 {bone1Name=dyn_lace_01_r, bone2Name=dyn_lace_r_MSTR, swing1LimitDegrees=30, swing2LimitDegrees=30, twistLimitDegrees=5}
]
