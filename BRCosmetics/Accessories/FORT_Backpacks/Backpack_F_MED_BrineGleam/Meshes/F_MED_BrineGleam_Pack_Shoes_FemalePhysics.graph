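# Physics asset for the female BrineGleam backpack shoes cosmetic mesh, defining capsule bodies for two dangling ribbon bone chains and the backpack spine.
# Constraints link each ribbon from its master bone, through three segments on the left and two on the right, with the left 01-to-02 joint nearly rigid.
# Each chain tip carries two large capsules, while unconstrained capsules on spine bones 02 to 05 form a torso-sized collision volume.
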PhysicsAsset /BRCosmetics/Accessories/FORT_Backpacks/Backpack_F_MED_BrineGleam/Meshes/F_MED_BrineGleam_Pack_Shoes_FemalePhysics
bodies {
 dyn_ribbon_01_l [{shapeType=Capsule, length=0.5, radius=1, shapeName=dyn_ribbon_01_l_capsule, center=<2.7847 -0.009 0.2002>}]
 dyn_ribbon_01_r [{shapeType=Capsule, length=0.5, radius=1.2217, shapeName=dyn_ribbon_01_r_capsule, center=<3.4096 -0.0316 -0.0649>}]
 dyn_ribbon_02_l [{shapeType=Capsule, length=1, radius=1, shapeName=dyn_ribbon_02_l_capsule, center=<0.5 0.0362 -0.0614>}]
 dyn_ribbon_02_r [
  {
   shapeType=Capsule
   length=16
   radius=5
   shapeName=dyn_ribbon_02_r_capsule
   center=<11.9293 2.7294 0.9452>
   rotation=<rot 4.3098 -129.1141 -102.6613>
  }
  {
   shapeType=Capsule
   length=0.9565
   radius=4.875
   shapeName=dyn_ribbon_02_r_capsule
   center=<3.5316 1.6093 -0.5217>
   rotation=<rot 77.8805 -144.2666 -9.754>
  }
 ]
 dyn_ribbon_03_l [
  {
   shapeType=Capsule
   length=15
   radius=5.25
   shapeName=dyn_ribbon_03_l_capsule
   center=<7.2239 2.5392 0.4877>
   rotation=<rot -13.6855 -125.4138 -72.3336>
  }
  {shapeType=Capsule, length=4.3619, radius=4.5, shapeName=dyn_ribbon_03_l_capsule, center=<0.7554 2.2024 1.1895>, rotation=<rot 90 -125.825 0>}
 ]
 dyn_ribbon_l_MSTR [{shapeType=Capsule, length=0.5, radius=0.5, shapeName=dyn_ribbon_l_MSTR_capsule}]
 dyn_ribbon_r_MSTR [{shapeType=Capsule, length=0.5, radius=0.5, shapeName=dyn_ribbon_r_MSTR_capsule}]
 spine_02_BrineGleam_backpack [
  {shapeType=Capsule, length=2.5219, radius=9.7317, center=<-2.6885 -1.2641 0>}
  {shapeType=Capsule, length=20.6875, radius=9.3125, center=<-1.8317 1.9362 0>, rotation=<rot 90 0 0>}
 ]
 spine_03_BrineGleam_backpack [
  {shapeType=Capsule, length=6.1863, radius=8.4854, center=<0.0574 0.4416 0>}
  {shapeType=Capsule, length=20.7188, radius=9.2812, center=<-1.2163 0.549 0>, rotation=<rot 90 180 -175.0732>}
 ]
 spine_04_BrineGleam_backpack [
  {shapeType=Capsule, length=8.7576, radius=10.6416, center=<5.6234 -0.6779 0>}
  {shapeType=Capsule, length=10.0312, radius=9.2812, center=<-1.2326 1.308 0>, rotation=<rot 90 0 0>}
 ]
 spine_05_BrineGleam_backpack [{shapeType=Capsule, length=8.678, radius=13.3571, center=<0.6969 1.91 0>}]
}
constraints [
 {bone1Name=dyn_ribbon_02_l, bone2Name=dyn_ribbon_01_l, swing1LimitDegrees=1, swing2LimitDegrees=1, twistLimitDegrees=0}
 {bone1Name=dyn_ribbon_02_r, bone2Name=dyn_ribbon_01_r, swing1LimitDegrees=30, swing2LimitDegrees=30, twistLimitDegrees=5}
 {bone1Name=dyn_ribbon_01_r, bone2Name=dyn_ribbon_r_MSTR, swing1LimitDegrees=30, swing2LimitDegrees=30, twistLimitDegrees=5}
 {bone1Name=dyn_ribbon_01_l, bone2Name=dyn_ribbon_l_MSTR, swing1LimitDegrees=45, swing2LimitDegrees=30, twistLimitDegrees=5}
 {bone1Name=dyn_ribbon_03_l, bone2Name=dyn_ribbon_02_l, swing1LimitDegrees=30, swing2LimitDegrees=20, twistLimitDegrees=5}
]
