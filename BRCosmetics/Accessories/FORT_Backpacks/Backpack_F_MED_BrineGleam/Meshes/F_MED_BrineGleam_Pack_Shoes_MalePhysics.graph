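# Physics asset for the BrineGleam backpack's shoes mesh, male variant, simulating dangling left and right ribbon chains.
# Each chain links from a master bone through swing- and twist-limited constraints, with the left chain one segment longer than the right.
# Small capsules cover the upper links, while paired larger capsules on each end bone form the bulkier hanging volumes.
# Unconstrained multi-capsule bodies on backpack spine bones approximate the torso, likely as collision for the chains.
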PhysicsAsset /BRCosmetics/Accessories/FORT_Backpacks/Backpack_F_MED_BrineGleam/Meshes/F_MED_BrineGleam_Pack_Shoes_MalePhysics
bodies {
 dyn_ribbon_01_l [{shapeType=Capsule, length=0.5, radius=1, shapeName=dyn_ribbon_01_l_capsule, center=<3 -0.009 -0.2002>}]
 dyn_ribbon_01_r [{shapeType=Capsule, length=0.5, radius=1, shapeName=dyn_ribbon_01_r_capsule, center=<3.4096 -0.0032 -0.0649>}]
 dyn_ribbon_02_l [{shapeType=Capsule, length=1, radius=1, shapeName=dyn_ribbon_02_l_capsule, center=<0.5 0.0362 -0.0614>}]
 dyn_ribbon_02_r [
  {
   shapeType=Capsule
   length=16
   radius=5
   shapeName=dyn_ribbon_02_r_capsule
   center=<12.2749 2.7294 0.9452>
   rotation=<rot 4.3098 -129.1141 -102.6613>
  }
  {
   shapeType=Capsule
   length=3
   radius=4.5
   shapeName=dyn_ribbon_02_r_capsule
   center=<4.0645 1.6093 -0.5217>
   rotation=<rot 77.8805 -144.2666 170.4655>
  }
 ]
 dyn_ribbon_03_l [
  {
   shapeType=Capsule
   length=15.5
   radius=5
   shapeName=dyn_ribbon_03_l_capsule
   center=<9.3533 2.5234 -0.8364>
   rotation=<rot -13.6855 -125.4139 -67.3336>
  }
  {
   shapeType=Capsule
   length=3.695
   radius=4.8744
   shapeName=dyn_ribbon_03_l_capsule
   center=<2.9931 2.2024 1.1895>
   rotation=<rot 66.5752 -39.4962 -102.1222>
  }
 ]
 dyn_ribbon_l_MSTR [{shapeType=Capsule, length=0.5, radius=0.5, shapeName=dyn_ribbon_l_MSTR_capsule}]
 dyn_ribbon_r_MSTR [{shapeType=Capsule, length=0.5, radius=0.5, shapeName=dyn_ribbon_r_MSTR_capsule}]
 spine_02_BrineGleam_backpack [
  {shapeType=Capsule, length=7.3264, radius=13.6322, center=<-1.711 -0.1851 0.0001>}
  {shapeType=Capsule, length=29.3956, radius=11.5807, center=<2.5024 0.4992 3.9825>, rotation=<rot 90 0.0049 0.0049>}
  {shapeType=Capsule, length=29.875, radius=11.1013, center=<2.4704 0.4991 -3.9825>, rotation=<rot 90 -0.0396 -0.0396>}
 ]
 spine_03_BrineGleam_backpack [{shapeType=Capsule, length=8.2107, radius=14.7031, center=<0.1445 0.2396 0.0001>}]
 spine_04_BrineGleam_backpack [{shapeType=Capsule, length=9.4404, radius=15.1325, center=<2.766 0.7765 0.0002>}]
 spine_05_BrineGleam_backpack [
  {shapeType=Capsule, length=8.0226, radius=15.695, center=<-2.2795 1.8109 0>}
  {shapeType=Capsule, length=13.9518, radius=13.9425, center=<-0.2916 0.4452 0>}
 ]
}
constraints [
 {bone1Name=dyn_ribbon_01_l, bone2Name=dyn_ribbon_l_MSTR, swing1LimitDegrees=30, swing2LimitDegrees=30, twistLimitDegrees=5}
 {bone1Name=dyn_ribbon_02_l, bone2Name=dyn_ribbon_01_l, swing1LimitDegrees=30, swing2LimitDegrees=30, twistLimitDegrees=5}
 {bone1Name=dyn_ribbon_01_r, bone2Name=dyn_ribbon_r_MSTR, swing1LimitDegrees=30, swing2LimitDegrees=30, twistLimitDegrees=5}
 {bone1Name=dyn_ribbon_02_r, bone2Name=dyn_ribbon_01_r, swing1LimitDegrees=30, swing2LimitDegrees=30, twistLimitDegrees=5}
 {bone1Name=dyn_ribbon_03_l, bone2Name=dyn_ribbon_02_l, swing1LimitDegrees=30, swing2LimitDegrees=30, twistLimitDegrees=5}
]
