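# Backpack AnimBP for the BrineGleam cosmetic, adding secondary motion to its bows, ribbons, laces and shoes.
# The event graph samples body spine socket transforms, sets gender/body-type flags and varies a gravity variable by skydive state.
# World-space spine bone overrides build a cached pose feeding chained RigidBody sims with female shoe, male shoe and bow physics assets.
# When bRigidBodyEnabled is false, an AnimDynamics spring chain replaces that path.
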AnimBlueprint /BRCosmetics/Accessories/FORT_Backpacks/Backpack_F_MED_BrineGleam/Meshes/F_MED_BrineGleam_Pack_AnimBP
target_skeleton=&/F_MED_BrineGleam_Pack_Skeleton
graphs [
 AnimGraph {nodes [
   "Transform (Modify) Bone - Bone: dyn_ribbon_03_l"!(ComponentPose, bAlphaBoolEnabled:bool=True, Translation="0,0,0", Rotation="41.126,-21.170,33.171", Scale="1,1,1") -> (Pose) {
    alphaInputType=Bool
    rotationMode=BMM_Additive
    rotationSpace=BCS_BoneSpace
    translationMode=BMM_Additive
    translationSpace=BCS_BoneSpace
    boneToModify {boneName=dyn_ribbon_03_l}
    rotation=<rot 41.1259 -21.1699 33.171>
   }
   n1 = Anim Dynamics - Bone: dyn_bow_01_r!(ComponentPose=Local To Component(LocalPose=Input Pose() {name=InPose_0}), Alpha=1, GravityOverride:Vector="0,0,0", ExternalForce="0,0,0") -> (Pose) {
    angularDampingOverride=0.9997
    angularSpringConstant=5000
    bAngularSpring=true
    bOverrideAngularDamping=true
    bUseGravityOverride=true
    boundBone {boneName=dyn_bow_01_r}
    componentAppliedLinearAccClamp=<1200 1200 1200>
    componentLinearAccScale=<0.5 0.5 0.5>
    componentLinearVelScale=<0.7 0.7 0.7>
    physicsBodyDefinitions [{
      boundBone {boneName=dyn_bow_01_r}
      boxExtents=<20 20 20>
      constraintSetup {angularLimitsMax=<5 30 30>, angularLimitsMin=<-5 -30 -30>, angularTarget=<1 0 0>}
      localJointOffset=<2 0 0>
     }]
   }
   n2 = Anim Dynamics - Bone: dyn_bow_02_r!(ComponentPose=@n1, Alpha=1, GravityOverride:Vector="0,0,0", ExternalForce="0,0,0") -> (Pose) {
    angularDampingOverride=0.9999
    angularSpringConstant=5000
    bAngularSpring=true
    bOverrideAngularDamping=true
    bUseGravityOverride=true
    boundBone {boneName=dyn_bow_02_r}
    componentAppliedLinearAccClamp=<1200 1200 1200>
    componentLinearAccScale=<0.5 0.5 0.5>
    componentLinearVelScale=<0.7 0.7 0.7>
    physicsBodyDefinitions [{
      boundBone {boneName=dyn_bow_02_r}
      boxExtents=<15 15 15>
      constraintSetup {angularLimitsMax=<5 30 30>, angularLimitsMin=<-5 -30 -30>, angularTarget=<1 0 0>}
      localJointOffset=<1 0 0>
     }]
   }
   n3 = Anim Dynamics - Bone: dyn_bow_01_l!(ComponentPose=@n2, Alpha=1, GravityOverride:Vector="0,0,0", ExternalForce="0,0,0") -> (Pose) {
    angularDampingOverride=0.998
    angularSpringConstant=7000
    bAngularSpring=true
    bOverrideAngularDamping=true
    bUseGravityOverride=true
    boundBone {boneName=dyn_bow_01_l}
    componentAppliedLinearAccClamp=<3500 3500 3500>
    componentLinearAccScale=<0.8 0.8 0.8>
    componentLinearVelScale=<1 1 1>
    physicsBodyDefinitions [{
      boundBone {boneName=dyn_bow_01_l}
      boxExtents=<25 25 25>
      constraintSetup {angularLimitsMax=<5 30 30>, angularLimitsMin=<-5 -30 -30>, angularTarget=<1 0 0>}
      localJointOffset=<1.5 0 0>
     }]
   }
   n4 = Anim Dynamics - Bone: dyn_bow_02_l!(ComponentPose=@n3, Alpha=1, GravityOverride:Vector="0,0,0", ExternalForce="0,0,0") -> (Pose) {
    angularDampingOverride=0.999
    angularSpringConstant=7000
    bAngularSpring=true
    bOverrideAngularDamping=true
    bUseGravityOverride=true
    boundBone {boneName=dyn_bow_02_l}
    componentAppliedLinearAccClamp=<3500 3500 3500>
    componentLinearAccScale=<0.8 0.8 0.8>
    componentLinearVelScale=<1 1 1>
    physicsBodyDefinitions [{
      boundBone {boneName=dyn_bow_02_l}
      boxExtents=<20 20 20>
      constraintSetup {angularLimitsMax=<5 30 30>, angularLimitsMin=<-5 -30 -30>, angularTarget=<1 0 0>}
      localJointOffset=<1 0 0>
     }]
   }
   n5 = Anim Dynamics - Bone: dyn_lace_01_l!(ComponentPose=@n4, Alpha=1, GravityOverride:Vector="0,0,0", ExternalForce="0,0,0") -> (Pose) {
    angularDampingOverride=0.997
    angularSpringConstant=500
    bAngularSpring=true
    bOverrideAngularDamping=true
    bUseGravityOverride=true
    boundBone {boneName=dyn_lace_01_l}
    componentAppliedLinearAccClamp=<1000 1000 1000>
    componentLinearAccScale=<0.4 0.4 0.4>
    componentLinearVelScale=<0.7 0.7 0.7>
    physicsBodyDefinitions [{
      boundBone {boneName=dyn_lace_01_l}
      boxExtents=<10 10 10>
      constraintSetup {angularLimitsMax=<2 15 20>, angularLimitsMin=<-2 -15 -5>, angularTarget=<1 0 0>}
      localJointOffset=<1.3 0 0>
     }]
   }
   n6 = Anim Dynamics - Bone: dyn_lace_01_r!(ComponentPose=@n5, Alpha=1, GravityOverride:Vector="0,0,0", ExternalForce="0,0,0") -> (Pose) {
    angularDampingOverride=0.999
    angularSpringConstant=1000
    bAngularSpring=true
    bOverrideAngularDamping=true
    bUseGravityOverride=true
    boundBone {boneName=dyn_lace_01_r}
    componentAppliedLinearAccClamp=<1000 1000 1000>
    componentLinearAccScale=<0.4 0.4 0.4>
    componentLinearVelScale=<0.7 0.7 0.7>
    physicsBodyDefinitions [{
      boundBone {boneName=dyn_lace_01_r}
      boxExtents=<10 10 5>
      constraintSetup {angularLimitsMax=<2 15 5>, angularLimitsMin=<-2 -15 -20>, angularTarget=<1 0 0>}
      localJointOffset=<1 0 0>
     }]
   }
   n7 = Anim Dynamics - Bone: dyn_ribbon_01_l!(ComponentPose=@n6, Alpha=1, GravityOverride:Vector="0,0,0", ExternalForce="0,0,0") -> (Pose) {
    angularDampingOverride=0.5
    angularSpringConstant=500
    bAngularSpring=true
    bOverrideAngularDamping=true
    bUseGravityOverride=true
    boundBone {boneName=dyn_ribbon_01_l}
    componentAppliedLinearAccClamp=<600 600 950>
    componentLinearAccScale=<0.5 0.5 0.5>
    componentLinearVelScale=<0.7 0.7 0.7>
    physicsBodyDefinitions [{
      boundBone {boneName=dyn_ribbon_01_l}
      boxExtents=<10 10 10>
      constraintSetup {angularLimitsMax=<2.5 0 -10>, angularLimitsMin=<-2.5 -30 -60>, angularTarget=<1 0 0>}
      localJointOffset=<1 0 0>
     }]
   }
   n8 = Anim Dynamics - Bone: dyn_ribbon_02_l!(ComponentPose=@n7, Alpha=1, GravityOverride:Vector="0,0,0", ExternalForce="0,0,0") -> (Pose) {
    angularDampingOverride=0.5
    angularSpringConstant=500
    bAngularSpring=true
    bOverrideAngularDamping=true
    bUseGravityOverride=true
    boundBone {boneName=dyn_ribbon_02_l}
    componentAppliedLinearAccClamp=<600 600 950>
    componentLinearAccScale=<0.5 0.5 0.5>
    componentLinearVelScale=<0.7 0.7 0.7>
    physicsBodyDefinitions [{
      boundBone {boneName=dyn_ribbon_02_l}
      boxExtents=<15 15 15>
      constraintSetup {angularLimitsMax=<2.5 0 -10>, angularLimitsMin=<-2.5 -30 -60>, angularTarget=<1 0 0>}
      localJointOffset=<1.5 0 0>
     }]
   }
   n9 = Anim Dynamics - Bone: dyn_ribbon_01_r!(ComponentPose=@n8, Alpha=1, GravityOverride:Vector="0,0,0", ExternalForce="0,0,0") -> (Pose) {
    angularDampingOverride=0.9
    angularSpringConstant=250
    bAngularSpring=true
    bOverrideAngularDamping=true
    bUseGravityOverride=true
    boundBone {boneName=dyn_ribbon_01_r}
    componentAppliedLinearAccClamp=<600 600 950>
    componentLinearAccScale=<0.5 0.5 0.5>
    componentLinearVelScale=<0.7 0.7 0.7>
    physicsBodyDefinitions [{
      boundBone {boneName=dyn_ribbon_01_r}
      boxExtents=<10 10 10>
      constraintSetup {angularLimitsMax=<2.5 30 -5>, angularLimitsMin=<-2.5 0 -50>, angularTarget=<1 0 0>}
      localJointOffset=<1 0 0>
     }]
   }
   n10 = Anim Dynamics - Bone: dyn_ribbon_02_r!(ComponentPose=@n9, Alpha=1, GravityOverride:Vector="0,0,0", ExternalForce="0,0,0") -> (Pose) {
    angularDampingOverride=0.9
    angularSpringConstant=250
    bAngularSpring=true
    bOverrideAngularDamping=true
    bUseGravityOverride=true
    boundBone {boneName=dyn_ribbon_02_r}
    componentAppliedLinearAccClamp=<600 600 950>
    componentLinearAccScale=<0.5 0.5 0.5>
    componentLinearVelScale=<0.7 0.7 0.7>
    physicsBodyDefinitions [{
      boundBone {boneName=dyn_ribbon_02_r}
      boxExtents=<10 10 10>
      constraintSetup {angularLimitsMax=<2.5 30 -5>, angularLimitsMin=<-2.5 0 -50>, angularTarget=<1 0 0>}
      localJointOffset=<1 0 0>
     }]
   }
   n11 = Anim Dynamics - Bone: dyn_ribbon_03_r!(ComponentPose=@n10, Alpha=1, GravityOverride:Vector="0,0,0", ExternalForce="0,0,0") -> (Pose) {
    angularDampingOverride=0.9
    angularSpringConstant=250
    bAngularSpring=true
    bOverrideAngularDamping=true
    bUseGravityOverride=true
    boundBone {boneName=dyn_ribbon_03_r}
    componentAppliedLinearAccClamp=<600 600 950>
    componentLinearAccScale=<0.5 0.5 0.5>
    componentLinearVelScale=<0.7 0.7 0.7>
    physicsBodyDefinitions [{
      boundBone {boneName=dyn_ribbon_03_r}
      boxExtents=<15 15 15>
      constraintSetup {angularLimitsMax=<2.5 30 -5>, angularLimitsMin=<-2.5 0 -50>, angularTarget=<1 0 0>}
      localJointOffset=<1.5 0 0>
     }]
   }
   n12 = "Transform (Modify) Bone - Bone: dyn_ribbon_l_MSTR"(
    ComponentPose="Transform (Modify) Bone - Bone: dyn_ribbon_r_MSTR"(ComponentPose=@n11, Alpha=1, Translation="0,0,0", Rotation="-0.000,-15.000,0.000", Scale="1,1,1") {rotationMode=BMM_Additive, boneToModify {boneName=dyn_ribbon_r_MSTR}, rotation {yaw=-15}}
    Alpha=1
    Translation="0,0,0"
    Rotation="0,15,0"
    Scale="1,1,1"
   ) {rotationMode=BMM_Additive, boneToModify {boneName=dyn_ribbon_l_MSTR}, rotation {yaw=15}}
   n13 = "Transform (Modify) Bone - Bone: dyn_ribbon_l_MSTR"!(
    ComponentPose=Local To Component(LocalPose=Use cached pose 'PreSim'())
    bAlphaBoolEnabled:bool=True
    Translation="0,0,0"
    Rotation="0,0,20"
    Scale="1,1,1"
   ) -> (Pose) {
    alphaInputType=Bool
    rotationMode=BMM_Additive
    rotationSpace=BCS_BoneSpace
    translationMode=BMM_Additive
    translationSpace=BCS_BoneSpace
    boneToModify {boneName=dyn_ribbon_l_MSTR}
    rotation {roll=20}
   }
   n14 = RigidBody!(ComponentPose=@n13, bAlphaBoolEnabled:bool, OverrideWorldGravity:Vector="0,0,-980", ExternalForce) -> (Pose) {
    alphaInputType=Bool
    bOverrideWorldGravity=true
    overridePhysicsAsset=&/F_MED_BrineGleam_Pack_Shoes_FemalePhysics
    simulationSpace=BaseBoneSpace
    alphaBoolBlend {blendInTime=0.5, blendOption=QuarticInOut, blendOutTime=0.5}
    baseBoneRef {boneName=pack}
    overrideWorldGravity {z=-980}
    simSpaceSettings {dampingAlpha=0, maxAngularAcceleration=32, maxAngularVelocity=9, maxLinearAcceleration=1200, maxLinearVelocity=700, worldAlpha=1}
   }
   n15 = RigidBody(ComponentPose=@n14, OverrideWorldGravity="0,0,-980") {
    alphaInputType=Bool
    bOverrideWorldGravity=true
    overridePhysicsAsset=&/F_MED_BrineGleam_Pack_Shoes_MalePhysics
    simulationSpace=BaseBoneSpace
    alphaBoolBlend {blendInTime=0.5, blendOption=QuarticInOut, blendOutTime=0.5}
    baseBoneRef {boneName=pack}
    overrideWorldGravity {z=-980}
    simSpaceSettings {dampingAlpha=0, maxAngularAcceleration=32, maxAngularVelocity=9, maxLinearAcceleration=1200, maxLinearVelocity=700, worldAlpha=1}
   }
   n16 = RigidBody!(ComponentPose=@n15, Alpha:float, OverrideWorldGravity="0,0,-980", ExternalForce) -> (Pose) {
    bOverrideWorldGravity=true
    overridePhysicsAsset=&/F_MED_BrineGleam_Pack_Bow_Physics
    simulationSpace=BaseBoneSpace
    alphaBoolBlend {blendInTime=0.5, blendOption=QuarticInOut, blendOutTime=0.5}
    baseBoneRef {boneName=pack}
    overrideWorldGravity {z=-980}
    simSpaceSettings {dampingAlpha=0, maxAngularAcceleration=40, maxAngularVelocity=8, maxLinearAcceleration=800, maxLinearVelocity=100, worldAlpha=1}
   }
   Output Pose(Result=Blend Poses by bool(BlendPose_0=Component To Local(ComponentPose=@n16), BlendPose_1=Component To Local(ComponentPose=@n12), bActiveValue=Get bRigidBodyEnabled())) {layerGroup=None}
   n17 = "Transform (Modify) Bone - Bone: spine_02_BrineGleam_backpack"(ComponentPose=Local To Component(LocalPose=Input Pose()), Alpha=1, Translation="0,0,0", Rotation="0,0,0", Scale="1,1,1") {
    rotationMode=BMM_Replace
    rotationSpace=BCS_WorldSpace
    translationMode=BMM_Replace
    translationSpace=BCS_WorldSpace
    boneToModify {boneName=spine_02_BrineGleam_backpack}
   }
   n18 = "Transform (Modify) Bone - Bone: spine_03_BrineGleam_backpack"(ComponentPose=@n17, Alpha=1, Translation="0,0,0", Rotation="0,0,0", Scale="1,1,1") {
    rotationMode=BMM_Replace
    rotationSpace=BCS_WorldSpace
    translationMode=BMM_Replace
    translationSpace=BCS_WorldSpace
    boneToModify {boneName=spine_03_BrineGleam_backpack}
   }
   n19 = "Transform (Modify) Bone - Bone: spine_04_BrineGleam_backpack"(ComponentPose=@n18, Alpha=1, Translation="0,0,0", Rotation="0,0,0", Scale="1,1,1") {
    rotationMode=BMM_Replace
    rotationSpace=BCS_WorldSpace
    translationMode=BMM_Replace
    translationSpace=BCS_WorldSpace
    boneToModify {boneName=spine_04_BrineGleam_backpack}
   }
   n20 = "Transform (Modify) Bone - Bone: spine_05_BrineGleam_backpack"(ComponentPose=@n19, Alpha=1, Translation="0,0,0", Rotation="0,0,0", Scale="1,1,1") {
    rotationMode=BMM_Replace
    rotationSpace=BCS_WorldSpace
    translationMode=BMM_Replace
    translationSpace=BCS_WorldSpace
    boneToModify {boneName=spine_05_BrineGleam_backpack}
   }
   Save cached pose 'PreSim'(Pose=Component To Local(ComponentPose=@n20)) {cacheName=PreSim}
  ]}
 EventGraph {nodes [
   Event BlueprintUpdateAnimation()
   > n1 = Sequence!(execute) -> (then_0, then_1, then_2:Exec)
   > Get!(execute:Exec) -> (then:Exec, else:Exec, BodyMesh:Skeletal Mesh Component Budgeted Object Reference)
   n2 = Get BodyMesh() -> (BodyMesh:Skeletal Mesh Component Budgeted Object Reference)
   n3 = BreakTransform(InTransform=GetSocketTransform(self=@n2, InSocketName=spine_02, TransformSpace=RTS_World))
   > Set Spine 02 Translation(Spine 02 Translation:Vector=@n3) -> (Output_Get:Vector)
   > Set Spine 02 Rotation(Spine 02 Rotation:Rotator=@n3.Rotation) -> (Output_Get:Rotator)
   n4 = BreakTransform(InTransform=GetSocketTransform(self=@n2, InSocketName=spine_03, TransformSpace=RTS_World))
   > Set Spine 03 Translation(Spine 03 Translation:Vector=@n4) -> (Output_Get:Vector)
   > Set Spine 03 Rotation(Spine 03 Rotation:Rotator=@n4.Rotation) -> (Output_Get:Rotator)
   n5 = BreakTransform(InTransform=GetSocketTransform(self=@n2, InSocketName=spine_04, TransformSpace=RTS_World))
   > Set Spine 04 Translation(Spine 04 Translation:Vector=@n5) -> (Output_Get:Vector)
   > Set Spine 04 Rotation(Spine 04 Rotation:Rotator=@n5.Rotation) -> (Output_Get:Rotator)
   n6 = BreakTransform(InTransform=GetSocketTransform(self=@n2, InSocketName=spine_05, TransformSpace=RTS_World))
   > Set Spine 05 Translation(Spine 05 Translation:Vector=@n6) -> (Output_Get:Vector)
   > Set Spine 05 Rotation(Spine 05 Rotation:Rotator=@n6.Rotation) -> (Output_Get:Rotator)
   n7 = Branch(execute=@n1.then_1, Condition=Get bIsSkydiveFloating())
   > Set WorldGravity(WorldGravity:Vector="0,-100,600") -> (Output_Get:Vector)
   n8 = Branch(execute=@n7.else, Condition=Get bIsSkydiveDiveMode())
   > Set WorldGravity(WorldGravity="-500,0,700")
   Set WorldGravity(execute=@n8.else, WorldGravity="0,0,-980")
   n9 = Event OnCharacterCustomizationLoaded()
   > Set BodyMesh(BodyMesh:Skeletal Mesh Component Budgeted Object Reference=Get Array Item with EFort Custom Part Type(Array=@n9.CustomMeshParts)) -> (Output_Get:Skeletal Mesh Component Budgeted Object Reference)
   n10 = Get Array Item with EFort Custom Part Type(Array:Array of Custom Character Part Object References=@n9.CustomParts) -> (Output:Custom Character Part Object Reference)
   n11 = "Equal (Enum)"(
    A:EFortCustomBodyType Enum=Get BodyTypesPermitted!(self:Custom Character Part Object Reference=@n10) -> (BodyTypesPermitted:EFortCustomBodyType Enum=NONE)
    B:EFortCustomBodyType Enum=Large
   )
   n12 = Get GenderPermitted!(self:Custom Character Part Object Reference=@n10) -> (GenderPermitted:EFortCustomGender Enum=Invalid)
   n13 = "Equal (Enum)"(A:EFortCustomGender Enum=@n12, B:EFortCustomGender Enum=Female)
   > Set IsMaleOrLargeFemale(IsMaleOrLargeFemale:bool=OR Boolean(A=AND Boolean(A=@n11, B=@n13), B="Equal (Enum)"(A=@n12, B=Male))) -> (Output_Get:bool)
   > Set IsFemale(IsFemale:bool=@n13) -> (Output_Get:bool)
  ]}
]
variables [
 BodyMesh {type=SkeletalMeshComponentBudgeted}
 IsFemale {type=boolean}
 IsMaleOrLargeFemale {type=boolean}
 Spine 02 Rotation {type=Rotator}
 Spine 02 Translation {type=Vector}
 Spine 03 Rotation {type=Rotator}
 Spine 03 Translation {type=Vector}
 Spine 04 Rotation {type=Rotator}
 Spine 04 Translation {type=Vector}
 Spine 05 Rotation {type=Rotator}
 Spine 05 Translation {type=Vector}
 WorldGravity {type=Vector}
]
